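# Niagara effect for the Sweet Toss emote prop that sprays colored candy sprites and meshes from a sampled skeletal mesh.
# CPU, GPU and Mobile emitters share spawning: SpawnIntensity scales rate and burst velocity, PourVelocity adds cone speed, and Colors selects one of three palette bands.
# All use gravity and drag; CPU adds collision, while GPU uses a bone-anchored blocking sphere instead.
# Debug sphere, directional burst and plane darkener emitters are disabled.
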NiagaraSystem /BRCosmetics/Animation/Props/Gadgets_Sweet_Toss/FX/NS_Emote_SweetToss
dependencies {
 usedDataInterfaces [&NiagaraDataInterfaceColorCurve, &NiagaraDataInterfaceArrayFloat, &NiagaraDataInterfaceCurve]
 usedDynamicInputs []
 usedModules [&SystemState]
}
emitters {
 Candy_CPU {
  emitterUpdateScript [
   &EmitterState()
   &SpawnRate(
    SpawnRate=&Lerp_Float(A=&Multiply_Float(A=$User.SpawnIntensity, B=100), B=50, Alpha=0)
    Use Spawn Probability=false
   ) {module=SpawnRate001}
  ]
  particleSpawnScript [
   &InitializeParticle(
    Color=&ColorFromCurve(
     ColorCurve=DataInterface() {propertyValues {
       AlphaCurve {keys [<at 0 1 RCIM_Constant>]}
       BlueCurve {keys [<at 0 1 RCIM_Constant>,<at 0.1739 0.0551 RCIM_Constant>,<at 0.2302 0.4418 RCIM_Constant>,<at 0.2857 0.9601 RCIM_Constant>,<at 0.33 0 RCIM_Constant>,<at 0.5399 0 RCIM_Constant>,<at 0.5978 0.6094 RCIM_Constant>,<at 0.66 0.533 RCIM_Constant>,<at 0.8835 0.13 RCIM_Constant>,<at 0.9481 1 RCIM_Constant>]}
       GreenCurve {keys [<at 0 0.8658 RCIM_Constant>,<at 0.1739 0.3212 RCIM_Constant>,<at 0.2302 0.0248 RCIM_Constant>,<at 0.2857 0.015 RCIM_Constant>,<at 0.33 0.5573 RCIM_Constant>,<at 0.5399 0.2124 RCIM_Constant>,<at 0.5978 0.836 RCIM_Constant>,<at 0.66 0.048 RCIM_Constant>,<at 0.8835 0.0009 RCIM_Constant>,<at 0.9481 0.54 RCIM_Constant>]}
       RedCurve {keys [<at 0 0 RCIM_Constant>,<at 0.1739 0.0139 RCIM_Constant>,<at 0.2302 0.8388 RCIM_Constant>,<at 0.2857 0.094 RCIM_Constant>,<at 0.33 1 RCIM_Constant>,<at 0.5399 1 RCIM_Constant>,<at 0.5978 1 RCIM_Constant>,<at 0.66 0.1225 RCIM_Constant>,<at 0.8835 0.5365 RCIM_Constant>,<at 0.9481 0.0381 RCIM_Constant>]}
      }}
     CurveIndex=&RandomRangeFloat(
      Minimum=&SelectFloatFromArray(
       Direct Array Index=&MakeIntFromFloat(Float=&Add_Float(A=$User.Colors, B=0))
       Direct Array Mode=::NewEnumerator1
       Float Selection Array=DataInterface() {propertyValues {FloatData [0,0.33,0.66]}}
      )
      Maximum=&SelectFloatFromArray(
       Direct Array Index=&MakeIntFromFloat(Float=&Add_Float(A=$User.Colors, B=0))
       Direct Array Mode=::NewEnumerator1
       Float Selection Array=DataInterface() {propertyValues {FloatData [0.32,0.65,1]}}
      )
     )
    )
    Lifetime=&RandomRangeFloat(Minimum=0.7, Maximum=1.3)
    Mesh Scale Mode=::NewEnumerator3
    Sprite Rotation Mode=::NewEnumerator2
    Sprite Size Mode=::NewEnumerator3
    Uniform Sprite Size=&RandomRangeFloat(Minimum=2, Maximum=7)
    UsePositionOffset=false
    Write Material Random=true
   )
   &SkeletalMeshLocation(
    Bone Sampling Mode=::NewEnumerator4
    MaskSpawnGroup=false
    Position Sampling=::NewEnumerator1
    Sampled Position Offset=&Add_Vector(
     A=<0 0 0>
     B=&Lerp_Vector(A=<5 -3 -9>, B=&RandomRangeVector(Minimum=<-12 0 20>, Maximum=<12 0 30>), Alpha=0)
    )
    Skeletal Mesh=$User.New Skeletal Mesh
   )
   &BoxLocation(Box Offset=<0 0 25>, Box Size=<25 5 10>, MaskSpawnGroup=false, Module.Override Seed=false, Rotate Box=false, Surface Only=false) {enabled=false}
   &AddVelocityFromPoint(
    Module.Use Falloff Distance=false
    Origin Offset=<20 0 -10>
    Velocity Origin:NiagaraPosition=$Particles.SkeletalMeshLocation.SampledPosition
    Velocity Strength=&Multiply_Float(A=100, B=$User.SpawnIntensity)
   )
   &AddVelocity(Velocity=&Multiply_VectorByFloat(
     Vector=&RotateVectorByQuaternion(
      Vector To Rotate=&RandomRangeVector(Minimum=<0 0 150>, Maximum=<0 0 300>)
      Quaternion=$Particles.SkeletalMeshLocation.SampledRotation
     )
     Float=$User.SpawnIntensity
    )) {enabled=false}
   /Niagara/Modules/Update/SubUV/SubUVAnimation(SubUV Animation Mode=::NewEnumerator1)
   &AddVelocity(
    Cone Angle=30
    Cone Axis=<2 0 0>
    Module.UseConeFalloff=false
    Module.UseVelDistribution=false
    Override Seed=false
    Velocity Mode=::NewEnumerator2
    Velocity Speed=&Multiply_Float(A=-120, B=$User.PourVelocity)
   ) {module=AddVelocity001}
   &InheritVelocity(
    Inherited Velocity Amount Scale=<0.75 0.75 0.75>
    Speed Limit=false
    UseSpeedThreshold=false
    Velocity Source=$Particles.SkeletalMeshLocation.SampledVelocity
   )
   &InitialMeshOrientation(Use Rotation Vector=false)
   /SetVariables(Particles.MeshIndex=&RandomRangeInt(Minimum=0, Maximum=2)) {isSetParameters=true, module=SetVariables_#1}
  ]
  particleUpdateScript [
   &ParticleState()
   &self:SphereBlockingVolume_01(Pivot Offset:Vector3f=<0 4 20>, Sphere Origin:Vector3f=/Niagara/DynamicInputs/Helpers/SimulationPosition(), Sphere Radius=10) {enabled=false, module=SphereBlockingVolume_01}
   &GravityForce(Gravity=<0 0 -700>)
   &Drag(Drag=&RandomRangeFloat(Minimum=1, Maximum=3), Use Linear Drag=true, Use Rotational Drag=false)
   &Collision(
    Age Colliding Particles=false
    Control Roll On Collision=true
    Module.Control Rest State Via Surface Normal=false
    Module.EnableMaxCollisionCount=false
    Module.Randomize Collision Normal Vector=true
    Randomize Collision Normal=0.25
    Restitution=&RandomRangeFloat(Minimum=0.25, Maximum=0.6)
    ScaleRateOfRotation=false
   )
   &SolveForcesAndVelocity(Clamp Velocity=false, Limit Acceleration=false)
   &ScaleSpriteSize(Uniform Curve Sprite Scale=DataInterface() {propertyValues {Curve {keys [{interpMode=RCIM_Cubic, tangentMode=RCTM_User, time=0, value=1}, {interpMode=RCIM_Cubic, tangentMode=RCTM_User, time=1, value=0}]}}})
   &SpriteRotationRate()
   &ScaleMeshSize(Scale Factor=&VectorFromFloat(Value=&FloatFromCurve(FloatCurve=DataInterface() {propertyValues {Curve {keys [
          {interpMode=RCIM_Cubic, tangentMode=RCTM_User, time=0, value=0}
          {interpMode=RCIM_Cubic, tangentMode=RCTM_User, time=0.1, value=0.95}
          {interpMode=RCIM_Cubic, tangentMode=RCTM_User, time=0.9, value=1}
          <at 1 0>
         ]}}})))
   /SetVariables() {isSetParameters=true, module=SetVariables_#2}
  ]
  rendererClasses [&NiagaraSpriteRendererProperties, &NiagaraMeshRendererProperties]
  renderers [
   {rendererClass=&NiagaraSpriteRendererProperties}
   {rendererClass=&NiagaraMeshRendererProperties, rendererIndex=1}
  ]
 }
 Candy_GPU {
  simTarget=GPUComputeSim
  emitterUpdateScript [
   &EmitterState()
   &SpawnRate(
    SpawnRate=&Lerp_Float(A=&Multiply_Float(A=$User.SpawnIntensity, B=300), B=300, Alpha=0)
    Use Spawn Probability=false
   ) {module=SpawnRate001}
  ]
  particleSpawnScript [
   &InitializeParticle(
    Color=&ColorFromCurve(
     ColorCurve=DataInterface() {propertyValues {
       AlphaCurve {keys [<at 0 1 RCIM_Constant>]}
       BlueCurve {keys [<at 0 1 RCIM_Constant>,<at 0.1739 0.0551 RCIM_Constant>,<at 0.2302 0.4418 RCIM_Constant>,<at 0.2857 0.9601 RCIM_Constant>,<at 0.33 0 RCIM_Constant>,<at 0.5399 0 RCIM_Constant>,<at 0.5978 0.6094 RCIM_Constant>,<at 0.66 0.533 RCIM_Constant>,<at 0.8835 0.13 RCIM_Constant>,<at 0.9481 1 RCIM_Constant>]}
       GreenCurve {keys [<at 0 0.8658 RCIM_Constant>,<at 0.1739 0.3212 RCIM_Constant>,<at 0.2302 0.0248 RCIM_Constant>,<at 0.2857 0.015 RCIM_Constant>,<at 0.33 0.5573 RCIM_Constant>,<at 0.5399 0.2124 RCIM_Constant>,<at 0.5978 0.836 RCIM_Constant>,<at 0.66 0.048 RCIM_Constant>,<at 0.8835 0.0009 RCIM_Constant>,<at 0.9481 0.54 RCIM_Constant>]}
       RedCurve {keys [<at 0 0 RCIM_Constant>,<at 0.1739 0.0139 RCIM_Constant>,<at 0.2302 0.8388 RCIM_Constant>,<at 0.2857 0.094 RCIM_Constant>,<at 0.33 1 RCIM_Constant>,<at 0.5399 1 RCIM_Constant>,<at 0.5978 1 RCIM_Constant>,<at 0.66 0.1225 RCIM_Constant>,<at 0.8835 0.5365 RCIM_Constant>,<at 0.9481 0.0381 RCIM_Constant>]}
      }}
     CurveIndex=&RandomRangeFloat(
      Minimum=&SelectFloatFromArray(
       Direct Array Index=&MakeIntFromFloat(Float=&Add_Float(A=$User.Colors, B=0))
       Direct Array Mode=::NewEnumerator1
       Float Selection Array=DataInterface() {propertyValues {FloatData [0,0.33,0.66]}}
      )
      Maximum=&SelectFloatFromArray(
       Direct Array Index=&MakeIntFromFloat(Float=&Add_Float(A=$User.Colors, B=0))
       Direct Array Mode=::NewEnumerator1
       Float Selection Array=DataInterface() {propertyValues {FloatData [0.32,0.65,1]}}
      )
     )
    )
    Lifetime=&RandomRangeFloat(Minimum=0.7, Maximum=1.3)
    Mesh Scale Mode=::NewEnumerator3
    Sprite Rotation Mode=::NewEnumerator2
    Sprite Size Mode=::NewEnumerator3
    Uniform Sprite Size=&RandomRangeFloat(Minimum=2, Maximum=7)
    UsePositionOffset=false
    Write Material Random=true
   )
   &SkeletalMeshLocation(
    Bone Sampling Mode=::NewEnumerator4
    MaskSpawnGroup=false
    Position Sampling=::NewEnumerator1
    Sampled Position Offset=&Add_Vector(
     A=<0 0 0>
     B=&Lerp_Vector(A=<5 -3 -9>, B=&RandomRangeVector(Minimum=<-12 0 20>, Maximum=<12 0 30>), Alpha=0)
    )
    Skeletal Mesh=$User.New Skeletal Mesh
   )
   &BoxLocation(Box Offset=<0 0 25>, Box Size=<25 5 10>, MaskSpawnGroup=false, Module.Override Seed=false, Rotate Box=false, Surface Only=false) {enabled=false}
   &AddVelocityFromPoint(
    Module.Use Falloff Distance=false
    Origin Offset=<20 0 -10>
    Velocity Origin:NiagaraPosition=$Particles.SkeletalMeshLocation.SampledPosition
    Velocity Strength=&Multiply_Float(A=100, B=$User.SpawnIntensity)
   )
   &AddVelocity(Velocity=&Multiply_VectorByFloat(
     Vector=&RotateVectorByQuaternion(
      Vector To Rotate=&RandomRangeVector(Minimum=<0 0 150>, Maximum=<0 0 300>)
      Quaternion=$Particles.SkeletalMeshLocation.SampledRotation
     )
     Float=$User.SpawnIntensity
    )) {enabled=false}
   &SubUVAnimation(SubUV Animation Mode=::NewEnumerator1)
   &AddVelocity(
    Cone Angle=30
    Cone Axis=<2 0 0>
    Module.UseConeFalloff=false
    Module.UseVelDistribution=false
    Override Seed=false
    Velocity Mode=::NewEnumerator2
    Velocity Speed=&Multiply_Float(A=-120, B=$User.PourVelocity)
   ) {module=AddVelocity001}
   &InheritVelocity(
    Inherited Velocity Amount Scale=<0.75 0.75 0.75>
    Speed Limit=false
    UseSpeedThreshold=false
    Velocity Source=$Particles.SkeletalMeshLocation.SampledVelocity
   )
   &InitialMeshOrientation(Use Rotation Vector=false)
   /SetVariables(Particles.MeshIndex=&RandomRangeInt(Minimum=0, Maximum=2)) {isSetParameters=true, module=SetVariables_#1}
  ]
  particleUpdateScript [
   &ParticleState()
   &SkeletalMeshLocation(Bone / Socket Index=1, Bone Sampling Mode=::NewEnumerator4, MaskSpawnGroup=false, Skeletal Mesh=$User.New Skeletal Mesh) {module=SkeletalMeshLocation001}
   &self:SphereBlockingVolume_01(Sphere Origin:Vector3f=$Particles.Initial.SkeletalMeshLocation001.SampledPosition, Sphere Radius=2.4) {module=SphereBlockingVolume_01}
   &GravityForce(Gravity=<0 0 -900>)
   &Drag(Drag=&RandomRangeFloat(Minimum=1, Maximum=3), Use Linear Drag=true, Use Rotational Drag=false)
   &Collision(
    Age Colliding Particles=false
    Analytical Collision Plane Position 1:Vector3f=<0 0 1920>
    Control Roll On Collision=true
    GPU Collision Type=::NewEnumerator3
    Module.Control Rest State Via Surface Normal=false
    Module.EnableMaxCollisionCount=false
    Module.Randomize Collision Normal Vector=true
    Perform 2 Analytical Collision Queries=false
    Randomize Collision Normal=0.25
    Restitution=&RandomRangeFloat(Minimum=0.25, Maximum=0.6)
    ScaleRateOfRotation=false
   ) {enabled=false}
   &SolveForcesAndVelocity(Clamp Velocity=false, Limit Acceleration=false)
   &ScaleSpriteSize(Uniform Curve Sprite Scale=DataInterface() {propertyValues {Curve {keys [{interpMode=RCIM_Cubic, tangentMode=RCTM_User, time=0, value=1}, {interpMode=RCIM_Cubic, tangentMode=RCTM_User, time=1, value=0}]}}})
   &SpriteRotationRate()
   &ScaleMeshSize(Scale Factor=&VectorFromFloat(Value=&FloatFromCurve(FloatCurve=DataInterface() {propertyValues {Curve {keys [
          {interpMode=RCIM_Cubic, tangentMode=RCTM_User, time=0, value=0}
          {interpMode=RCIM_Cubic, tangentMode=RCTM_User, time=0.1, value=0.95}
          {interpMode=RCIM_Cubic, tangentMode=RCTM_User, time=0.9, value=1}
          <at 1 0>
         ]}}})))
   /SetVariables() {isSetParameters=true, module=SetVariables_#2}
  ]
  rendererClasses [&NiagaraSpriteRendererProperties, &NiagaraMeshRendererProperties]
  renderers [
   {rendererClass=&NiagaraSpriteRendererProperties}
   {rendererClass=&NiagaraMeshRendererProperties, rendererIndex=1}
  ]
 }
 Candy_Mobile {
  emitterUpdateScript [
   &EmitterState()
   &SpawnRate(
    SpawnRate=&Lerp_Float(A=&Multiply_Float(A=$User.SpawnIntensity, B=100), B=50, Alpha=0)
    Use Spawn Probability=false
   ) {module=SpawnRate001}
  ]
  particleSpawnScript [
   &InitializeParticle(
    Color=&ColorFromCurve(
     ColorCurve=DataInterface() {propertyValues {
       AlphaCurve {keys [<at 0 1 RCIM_Constant>]}
       BlueCurve {keys [<at 0 1 RCIM_Constant>,<at 0.1739 0.0551 RCIM_Constant>,<at 0.2302 0.4418 RCIM_Constant>,<at 0.2857 0.9601 RCIM_Constant>,<at 0.33 0 RCIM_Constant>,<at 0.5399 0 RCIM_Constant>,<at 0.5978 0.6094 RCIM_Constant>,<at 0.66 0.533 RCIM_Constant>,<at 0.8835 0.13 RCIM_Constant>,<at 0.9481 1 RCIM_Constant>]}
       GreenCurve {keys [<at 0 0.8658 RCIM_Constant>,<at 0.1739 0.3212 RCIM_Constant>,<at 0.2302 0.0248 RCIM_Constant>,<at 0.2857 0.015 RCIM_Constant>,<at 0.33 0.5573 RCIM_Constant>,<at 0.5399 0.2124 RCIM_Constant>,<at 0.5978 0.836 RCIM_Constant>,<at 0.66 0.048 RCIM_Constant>,<at 0.8835 0.0009 RCIM_Constant>,<at 0.9481 0.54 RCIM_Constant>]}
       RedCurve {keys [<at 0 0 RCIM_Constant>,<at 0.1739 0.0139 RCIM_Constant>,<at 0.2302 0.8388 RCIM_Constant>,<at 0.2857 0.094 RCIM_Constant>,<at 0.33 1 RCIM_Constant>,<at 0.5399 1 RCIM_Constant>,<at 0.5978 1 RCIM_Constant>,<at 0.66 0.1225 RCIM_Constant>,<at 0.8835 0.5365 RCIM_Constant>,<at 0.9481 0.0381 RCIM_Constant>]}
      }}
     CurveIndex=&RandomRangeFloat(
      Minimum=&SelectFloatFromArray(
       Direct Array Index=&MakeIntFromFloat(Float=&Add_Float(A=$User.Colors, B=0))
       Direct Array Mode=::NewEnumerator1
       Float Selection Array=DataInterface() {propertyValues {FloatData [0,0.33,0.66]}}
      )
      Maximum=&SelectFloatFromArray(
       Direct Array Index=&MakeIntFromFloat(Float=&Add_Float(A=$User.Colors, B=0))
       Direct Array Mode=::NewEnumerator1
       Float Selection Array=DataInterface() {propertyValues {FloatData [0.32,0.65,1]}}
      )
     )
    )
    Lifetime=&RandomRangeFloat(Minimum=0.7, Maximum=1.3)
    Mesh Scale Mode=::NewEnumerator3
    Sprite Rotation Mode=::NewEnumerator2
    Sprite Size Mode=::NewEnumerator3
    Uniform Sprite Size=&RandomRangeFloat(Minimum=2, Maximum=7)
    UsePositionOffset=false
    Write Material Random=true
   )
   &SkeletalMeshLocation(
    Bone Sampling Mode=::NewEnumerator4
    MaskSpawnGroup=false
    Position Sampling=::NewEnumerator1
    Sampled Position Offset=&Add_Vector(
     A=<0 0 0>
     B=&Lerp_Vector(A=<5 -3 -9>, B=&RandomRangeVector(Minimum=<-12 0 20>, Maximum=<12 0 30>), Alpha=0)
    )
    Skeletal Mesh=$User.New Skeletal Mesh
   )
   &BoxLocation(Box Offset=<0 0 25>, Box Size=<25 5 10>, MaskSpawnGroup=false, Module.Override Seed=false, Rotate Box=false, Surface Only=false) {enabled=false}
   &AddVelocityFromPoint(
    Module.Use Falloff Distance=false
    Origin Offset=<20 0 -10>
    Velocity Origin:NiagaraPosition=$Particles.SkeletalMeshLocation.SampledPosition
    Velocity Strength=&Multiply_Float(A=100, B=$User.SpawnIntensity)
   )
   &AddVelocity(Velocity=&Multiply_VectorByFloat(
     Vector=&RotateVectorByQuaternion(
      Vector To Rotate=&RandomRangeVector(Minimum=<0 0 150>, Maximum=<0 0 300>)
      Quaternion=$Particles.SkeletalMeshLocation.SampledRotation
     )
     Float=$User.SpawnIntensity
    )) {enabled=false}
   &SubUVAnimation(SubUV Animation Mode=::NewEnumerator1)
   &AddVelocity(
    Cone Angle=30
    Cone Axis=<2 0 0>
    Module.UseConeFalloff=false
    Module.UseVelDistribution=false
    Override Seed=false
    Velocity Mode=::NewEnumerator2
    Velocity Speed=&Multiply_Float(A=-120, B=$User.PourVelocity)
   ) {module=AddVelocity001}
   &InheritVelocity(
    Inherited Velocity Amount Scale=<0.75 0.75 0.75>
    Speed Limit=false
    UseSpeedThreshold=false
    Velocity Source=$Particles.SkeletalMeshLocation.SampledVelocity
   )
   &InitialMeshOrientation(Use Rotation Vector=false)
   /SetVariables(Particles.MeshIndex=&RandomRangeInt(Minimum=0, Maximum=2)) {isSetParameters=true, module=SetVariables_#1}
  ]
  particleUpdateScript [
   &ParticleState()
   &self:SphereBlockingVolume_01(Pivot Offset:Vector3f=<0 4 20>, Sphere Origin:Vector3f=&SimulationPosition(), Sphere Radius=10) {enabled=false, module=SphereBlockingVolume_01}
   &GravityForce(Gravity=<0 0 -700>)
   &Drag(Drag=&RandomRangeFloat(Minimum=1, Maximum=3), Use Linear Drag=true, Use Rotational Drag=false)
   &Collision(
    Age Colliding Particles=false
    Control Roll On Collision=true
    Module.Control Rest State Via Surface Normal=false
    Module.EnableMaxCollisionCount=false
    Module.Randomize Collision Normal Vector=true
    Randomize Collision Normal=0.25
    Restitution=&RandomRangeFloat(Minimum=0.25, Maximum=0.6)
    ScaleRateOfRotation=false
   ) {enabled=false}
   &SolveForcesAndVelocity(Clamp Velocity=false, Limit Acceleration=false)
   &ScaleSpriteSize(Uniform Curve Sprite Scale=DataInterface() {propertyValues {Curve {keys [{interpMode=RCIM_Cubic, tangentMode=RCTM_User, time=0, value=1}, {interpMode=RCIM_Cubic, tangentMode=RCTM_User, time=1, value=0}]}}})
   &SpriteRotationRate()
   &ScaleMeshSize(Scale Factor=&VectorFromFloat(Value=&FloatFromCurve(FloatCurve=DataInterface() {propertyValues {Curve {keys [
          {interpMode=RCIM_Cubic, tangentMode=RCTM_User, time=0, value=0}
          {interpMode=RCIM_Cubic, tangentMode=RCTM_User, time=0.1, value=0.95}
          {interpMode=RCIM_Cubic, tangentMode=RCTM_User, time=0.9, value=1}
          <at 1 0>
         ]}}})))
   /SetVariables() {enabled=false, isSetParameters=true, module=SetVariables_#2}
  ]
  rendererClasses [&NiagaraSpriteRendererProperties, &NiagaraMeshRendererProperties]
  renderers [
   {rendererClass=&NiagaraSpriteRendererProperties}
   {rendererClass=&NiagaraMeshRendererProperties, rendererIndex=1}
  ]
 }
 Debug_Sphere {
  bEnabled=false
  emitterUpdateScript [
   &EmitterState(Life Cycle Mode=::NewEnumerator1, Loop Behavior=::NewEnumerator1, Loop Duration=2, UseLoopDelay=false)
   &SpawnBurst_Instantaneous(Use Spawn Probability=false, UseLoopCountLimit=false)
  ]
  particleSpawnScript [
   &InitializeParticle(Lifetime=2, Sprite Size Mode=::NewEnumerator3, Uniform Sprite Size=400, UsePositionOffset=false, Write Material Random=true)
   &SphereLocation(
    MaskSpawnGroup=false
    Offset=<0 35 -100>
    Override Seed=false
    Sphere Radius=20
    Surface Only=false
    UseNonUniformScale=false
    UseOffset=true
    UseSphereOrientation=false
   )
  ]
  particleUpdateScript [
   &ParticleState(Kill Particles When Lifetime Has Elapsed=false)
   &SkeletalMeshLocation(Bone / Socket Index=1, Bone Sampling Mode=::NewEnumerator4, MaskSpawnGroup=false, Skeletal Mesh=$User.New Skeletal Mesh)
   &SolveForcesAndVelocity(Clamp Velocity=false, Limit Acceleration=false)
   /SetVariables(Particles.New Position:NiagaraPosition=&Add_Vector(A=&SimulationPosition(), B=<0 0 20>)) {enabled=false, isSetParameters=true}
  ]
  rendererClasses [&NiagaraSpriteRendererProperties]
  renderers [{rendererClass=&NiagaraSpriteRendererProperties}]
 }
 DirectionalBurst {
  bEnabled=false
  emitterUpdateScript [
   &EmitterState(Life Cycle Mode=::NewEnumerator1, Loop Behavior=::NewEnumerator1, Loop Duration=1, UseLoopDelay=false)
   &SpawnBurst_Instantaneous(Spawn Count=75, Use Spawn Probability=false, UseLoopCountLimit=false)
  ]
  particleSpawnScript [
   &InitializeParticle(
    Lifetime Max=1.75
    Lifetime Min=0.8
    Lifetime Mode=::NewEnumerator1
    Mass Max=2
    Mass Mode=::NewEnumerator3
    Sprite Size Max=<3 14>
    Sprite Size Min=<2 6>
    Sprite Size Mode=::NewEnumerator2
    UsePositionOffset=false
    Write Material Random=true
   )
   &AddVelocityInCone(
    Cone Angle=48
    Cone Axis=<0.9 0 0.4>
    Module.Use Velocity Falloff On Cone Axis=true
    Velocity Distribution Along Cone Axis=0.75
    Velocity Falloff Away From Cone Axis=0.333
    Velocity Strength=&RandomRangeFloat(Minimum=350, Maximum=750)
   )
  ]
  particleUpdateScript [
   &ParticleState()
   &GravityForce()
   &Drag(Drag=&RandomRangeFloat(Minimum=0.8, Maximum=1.2), Use Linear Drag=true, Use Rotational Drag=false)
   &ScaleColor(
    Scale Alpha=&FloatFromCurve(FloatCurve=DataInterface() {propertyValues {Curve {keys [<at -0.0046 1.0113 RCIM_Cubic>,<at 1.0011 0.002 RCIM_Cubic>]}}})
    ScaleA=true
    ScaleRGB=true
   )
   &SolveForcesAndVelocity(Clamp Velocity=false, Limit Acceleration=false)
   &ScaleSpriteSizeBySpeed(
    Max Scale Factor=<1.2 3.3>
    Min Scale Factor=<0.75 0.3>
    Sample Scale Factor By Curve=true
    Scale Factor Curve=DataInterface() {propertyValues {Curve {keys [
        {arriveTangent=2.865, interpMode=RCIM_Cubic, time=0.1247, value=-0.0006}
        {arriveTangent=4.1522, interpMode=RCIM_Cubic, leaveTangent=4.1522, tangentMode=RCTM_User, time=0.3948, value=0.4893}
        <at 0.9958 0.9974 RCIM_Cubic>
       ]}}}
    Velocity Threshold=1500
   )
  ]
  rendererClasses [&NiagaraSpriteRendererProperties]
  renderers [{rendererClass=&NiagaraSpriteRendererProperties}]
 }
 PlaneDarkener {
  bEnabled=false
  emitterUpdateScript [
   &EmitterState(Life Cycle Mode=::NewEnumerator1, Loop Behavior=::NewEnumerator1, Loop Duration=1, UseLoopDelay=false)
   &SpawnBurst_Instantaneous(Use Spawn Probability=false, UseLoopCountLimit=false)
  ]
  particleSpawnScript [
   &InitializeParticle(Mesh Scale Mode=::NewEnumerator3, Mesh Uniform Scale=80, UsePositionOffset=false, Write Material Random=true)
   &InitialMeshOrientation(Mesh Orientation Mode=::NewEnumerator7, Rotation=<0 -0.1 0>, Rotation Coordinate Space=::NewEnumerator2, Use Rotation Vector=true) {enabled=false}
  ]
  particleUpdateScript [
   &ParticleState(Kill Particles When Lifetime Has Elapsed=false)
   &SkeletalMeshLocation(Bone Sampling Mode=::NewEnumerator4, MaskSpawnGroup=false, Sampled Position Offset=<10 0 -10>, Skeletal Mesh=$User.New Skeletal Mesh)
   &ScaleColor(
    Scale Alpha=&FloatFromCurve(FloatCurve=DataInterface() {propertyValues {Curve {keys [<at 0.5 1.0113 RCIM_Cubic>,<at 1.0011 0.002 RCIM_Cubic>]}}})
    ScaleA=true
    ScaleRGB=false
   )
   &SolveForcesAndVelocity(Clamp Velocity=false, Limit Acceleration=false)
   &UpdateMeshOrientation(
    Facing Direction=<0 1 0>
    Facing Interpolation=false
    Orientation Method=::NewEnumerator1
    Override Quat=false
    TransformFacing=false
    TransformUp=false
    Up Direction:Vector3f=<0 0 0>
   ) {enabled=false}
  ]
  rendererClasses [&NiagaraMeshRendererProperties]
  renderers [{rendererClass=&NiagaraMeshRendererProperties, rendererIndex=0}]
 }
}
summary {
 systemName=NS_Emote_SweetToss
 emitters [
  {
   emitterName=Candy_GPU
   simTarget=GPUComputeSim
   rendererClasses [&NiagaraSpriteRendererProperties, &NiagaraMeshRendererProperties]
  }
  {bEnabled=false, emitterName=PlaneDarkener, rendererClasses [&NiagaraMeshRendererProperties]}
  {emitterName=Candy_CPU, rendererClasses [&NiagaraSpriteRendererProperties, &NiagaraMeshRendererProperties]}
  {bEnabled=false, emitterName=Debug_Sphere, rendererClasses [&NiagaraSpriteRendererProperties]}
  {emitterName=Candy_Mobile, rendererClasses [&NiagaraSpriteRendererProperties, &NiagaraMeshRendererProperties]}
  {bEnabled=false, emitterName=DirectionalBurst, rendererClasses [&NiagaraSpriteRendererProperties]}
 ]
 userVariables [
  User.BounceIntensity {type=NiagaraFloat, defaultValue=0, description=""}
  User.Colors {type=NiagaraFloat, defaultValue=0, description=""}
  User.New Skeletal Mesh {
   type=NiagaraDataInterfaceSkeletalMesh
   defaultValue {
    struct=NiagaraExt_VariableValue_DataInterface
    value {dataInterface=&self:NiagaraDataInterfaceSkeletalMesh_0}
   }
  }
  User.PourVelocity {type=NiagaraFloat, defaultValue=0, description=""}
  User.SpawnIntensity {type=NiagaraFloat, defaultValue=0.2, description=""}
 ]
}
