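# Niagara burst for the Sweet Toss gadget emote that throws a handful of randomly chosen mesh pieces from a user-supplied skeletal mesh.
# User.Switch scales the burst count, and launch speed along a narrow cone differs between gameplay and front end.
# Pieces pop in with squash-and-stretch curves, tumble under random gravity and drag, then shrink out.
# A near-duplicate Mobile emitter disables the collision bounce; the Empty emitter has no spawn module.
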NiagaraSystem /BRCosmetics/Animation/Props/Gadgets_Sweet_Toss/FX/NS_Emote_SweetToss_Burst
dependencies {usedDataInterfaces [&NiagaraDataInterfaceCurve], usedDynamicInputs [], usedModules [&SystemState]}
emitters {
 Empty {
  emitterUpdateScript [&EmitterState()]
  particleSpawnScript [&InitializeParticle(Color Mode=::NewEnumerator0, Lifetime=5, UsePositionOffset=false, Write Material Random=true)]
  particleUpdateScript [&ParticleState()]
  rendererClasses [&NiagaraSpriteRendererProperties]
  renderers [{rendererClass=&NiagaraSpriteRendererProperties}]
 }
 SweetToss_Burst {
  emitterUpdateScript [
   &EmitterState(Life Cycle Mode=::NewEnumerator1, Loop Behavior=::NewEnumerator1, Loop Duration=1, UseLoopDelay=false)
   &SpawnBurst_Instantaneous(
    Spawn Count=&MakeIntFromFloat(Float=&Lerp_Float(A=5, B=10, Alpha=$User.Switch))
    Use Spawn Probability=false
   )
  ]
  particleSpawnScript [
   /Niagara/Modules/Spawn/Initialization/InitializeParticle(
    Lifetime=&RandomRangeFloat(Minimum=0.5, Maximum=2)
    Mass=&RandomRangeFloat(Minimum=0.75, Maximum=1.75)
    Mesh Scale=&VectorFromFloat(Value=&Multiply_Float(A=&RandomRangeFloat(Minimum=0.3, Maximum=0.7), B=1))
    Module.Write Color=true
    Module.Write Lifetime=true
    Module.Write Mass=true
    Module.Write Material Random=true
    Module.Write Position=true
    Module.Write Scale=true
    Module.Write SpriteRotation=false
    Module.Write SpriteSize=false
    Position=&Add_Vector(A=/Niagara/DynamicInputs/Helpers/SimulationPosition(), B=<0 0 0>)
   )
   /SetVariables(Particles.MeshIndex=&RandomRangeInt(Minimum=0, Maximum=2)) {isSetParameters=true, module=SetVariables_#1}
   /Niagara/Modules/Update/Mass/CalculateSizeAndRotationalInertiaByMass(Depth=0, Height=5, Module.Manually Enter Density=false)
   &SkeletalMeshLocation(
    Bone Sampling Mode=::NewEnumerator4
    MaskSpawnGroup=false
    Position Sampling=::NewEnumerator1
    Sampled Position Offset=<0 0 20>
    Sampled Position Offset Coordinate Space:ENiagaraSkelMeshTransforms=::NewEnumerator2
    Skeletal Mesh=$User.New Skeletal Mesh
   )
   &SphereLocation(
    Hemisphere X=true
    MaskSpawnGroup=false
    Non Uniform Scale=<2.5 1 1>
    Override Seed=false
    Sphere Radius=2
    Surface Only=true
    UseNonUniformScale=true
    UseOffset=false
    UseSphereOrientation=false
   ) {enabled=false}
   &AddVelocity(
    Cone Angle=2
    Cone Axis=<0.4 -0.75 2>
    Module.UseConeFalloff=false
    Module.UseVelDistribution=false
    Override Seed=false
    Velocity Mode=::NewEnumerator2
    Velocity Speed=&FN_FrontEndFloatModifier(Gameplay=500, Front End=420)
   ) {module=AddVelocity002}
   &AddVelocity(Coordinate Space=::Simulation, Velocity=<150 -200 300>) {enabled=false}
   &AddVelocity(Velocity=&Multiply_VectorByFloat(
     Vector=&TransformVector(Vector=$Output.SphereLocation.SphereNormal, Source Space=::World, Destination Space=::Local)
     Float=&RandomRangeFloat(Minimum=40, Maximum=180)
    )) {enabled=false, module=AddVelocity001}
   /Niagara/Modules/Update/SubUV/SubUVAnimation(SubUV Animation Mode=::NewEnumerator1) {enabled=false}
   /Niagara/Modules/Spawn/Orientation/V2/InitialMeshOrientation(
    Rotation=&RandomRangeVector(Minimum=<-1 -1 -1>, Maximum=<1 1 1>)
    Use Orientation Vector=true
    Use Rotation Vector=true
   ) {enabled=false}
   &AddRotationalVelocity(
    Coordinate Space:ENiagaraMeshTransforms=::NewEnumerator2
    Rotation Rate=&RandomRangeVector(Minimum=<-0.7 -0.7 -0.7>, Maximum=<0.7 0.7 0.7>)
   )
   &DynamicMaterialParameters(
    Index 0 Param 1=&RandomRangeFloat(Minimum=0, Maximum=1)
    Module.Param0WriteEnabled=true
    Module.Param1WriteEnabled=false
    Module.Param2WriteEnabled=false
    Module.Param3WriteEnabled=false
   )
   &InitialMeshOrientation(
    Orientation Coordinate Space:ENiagaraCoordinateSpace=::World
    Orientation Vector=$Particles.Velocity
    Use Orientation Vector=true
    Use Rotation Vector=false
   ) {enabled=false, module=InitialMeshOrientation001}
   /Niagara/Modules/Spawn/Orientation/V3/InitialMeshOrientation(Use Rotation Vector=false) {module=InitialMeshOrientation002}
  ]
  particleUpdateScript [
   &ParticleState()
   &ScaleMeshSize(Scale Factor=&MakeVector(
     X=&FloatFromCurve(FloatCurve=DataInterface() {propertyValues {Curve {keys [
          {arriveTangent=26.0828, interpMode=RCIM_Cubic, leaveTangent=26.0828, tangentMode=RCTM_User, time=0, value=0.2}
          {arriveTangent=0.0334, interpMode=RCIM_Cubic, leaveTangent=0.0334, tangentMode=RCTM_User, time=0.0394, value=4.0724}
          {arriveTangent=-2.5074, interpMode=RCIM_Cubic, leaveTangent=-2.5074, tangentMode=RCTM_User, time=0.1173, value=0.7442}
          {arriveTangent=-0.1609, interpMode=RCIM_Cubic, leaveTangent=-0.1609, tangentMode=RCTM_User, time=0.2152, value=1.1946}
          {arriveTangent=0.1391, interpMode=RCIM_Cubic, leaveTangent=0.1391, tangentMode=RCTM_User, time=0.3123, value=0.8985}
          <at 0.35 1>
         ]}}})
     Y=&FloatFromCurve(FloatCurve=DataInterface() {propertyValues {Curve {keys [
          <at -0.003 1.0812 RCIM_Cubic>
          {arriveTangent=3.927, interpMode=RCIM_Cubic, leaveTangent=3.927, time=0.0492, value=1.1749}
          {arriveTangent=-2.0352, interpMode=RCIM_Cubic, leaveTangent=-2.0352, time=0.1219, value=1.5716}
          {arriveTangent=-0.5182, interpMode=RCIM_Cubic, leaveTangent=-0.5182, tangentMode=RCTM_User, time=0.2148, value=0.8379}
          {arriveTangent=-0.6419, interpMode=RCIM_Cubic, leaveTangent=-0.6419, tangentMode=RCTM_User, time=0.3121, value=1.1071}
          <at 0.35 1 RCIM_Cubic>
         ]}}})
     Z=&FloatFromCurve(FloatCurve=DataInterface() {propertyValues {Curve {keys [
          <at -0.0016 1.0762 RCIM_Cubic>
          {arriveTangent=-0.5131, interpMode=RCIM_Cubic, leaveTangent=-0.5131, tangentMode=RCTM_User, time=0.0502, value=1.1706}
          {arriveTangent=-2.0432, interpMode=RCIM_Cubic, leaveTangent=-2.0432, time=0.1212, value=1.571}
          {arriveTangent=0.2115, interpMode=RCIM_Cubic, leaveTangent=0.2115, tangentMode=RCTM_User, time=0.2145, value=0.835}
          {arriveTangent=-0.2737, interpMode=RCIM_Cubic, leaveTangent=-0.2737, tangentMode=RCTM_User, time=0.312, value=1.1039}
          <at 0.35 1>
         ]}}})
    )) {module=ScaleMeshSize001}
   &GravityForce(Gravity=&RandomRangeVector(Minimum=<0 0 -500>, Maximum=<1 1 -1200>))
   &Drag(Drag=&RandomRangeFloat(Minimum=2, Maximum=3), Use Linear Drag=true, Use Rotational Drag=false)
   &ScaleColor(
    Scale Alpha=&FloatFromCurve(FloatCurve=DataInterface() {propertyValues {Curve {keys [<at -0.0046 1.0113 RCIM_Cubic>,<at 1.0011 0.002 RCIM_Cubic>]}}})
    Scale RGB=<0 0 0>
    ScaleA=true
    ScaleRGB=true
   )
   &self:SphereBlockingVolume(Pivot Offset:Vector3f=<0 0 10>, Sphere Origin:Vector3f=&SimulationPosition(), Sphere Radius=5) {enabled=false, module=SphereBlockingVolume}
   &Collision(
    Age Colliding Particles=false
    Control Roll On Collision=true
    Module.Control Rest State Via Surface Normal=false
    Module.EnableMaxCollisionCount=false
    Module.Randomize Collision Normal Vector=false
    Restitution=&RandomRangeFloat(Minimum=0.25, Maximum=0.6)
    ScaleRateOfRotation=false
   )
   &SolveForcesAndVelocity(Clamp Velocity=false, Limit Acceleration=false)
   &SolveRotationalForcesAndVelocity()
   &ScaleMeshSize(Scale Factor=&VectorFromFloat(Value=&FloatFromCurve(FloatCurve=DataInterface() {propertyValues {Curve {keys [<at 0.75 1 RCIM_Cubic>, {arriveTangent=-8.6407, interpMode=RCIM_Cubic, leaveTangent=-8.6407, tangentMode=RCTM_User, time=1, value=0}]}}})))
  ]
  rendererClasses [&NiagaraMeshRendererProperties, &NiagaraSpriteRendererProperties]
  renderers [
   {rendererClass=&NiagaraMeshRendererProperties, rendererIndex=0}
   {rendererClass=&NiagaraSpriteRendererProperties, rendererIndex=1}
   {rendererClass=&NiagaraMeshRendererProperties, rendererIndex=2}
  ]
 }
 SweetToss_Burst_Mobile {
  emitterUpdateScript [
   &EmitterState(Life Cycle Mode=::NewEnumerator1, Loop Behavior=::NewEnumerator1, Loop Duration=1, UseLoopDelay=false)
   &SpawnBurst_Instantaneous(
    Spawn Count=&MakeIntFromFloat(Float=&Lerp_Float(A=5, B=10, Alpha=$User.Switch))
    Use Spawn Probability=false
   )
  ]
  particleSpawnScript [
   &InitializeParticle(
    Lifetime=&RandomRangeFloat(Minimum=0.5, Maximum=2)
    Mass=&RandomRangeFloat(Minimum=0.75, Maximum=1.75)
    Mesh Scale=&VectorFromFloat(Value=&Multiply_Float(A=&RandomRangeFloat(Minimum=0.3, Maximum=0.7), B=1))
    Module.Write Color=true
    Module.Write Lifetime=true
    Module.Write Mass=true
    Module.Write Material Random=true
    Module.Write Position=true
    Module.Write Scale=true
    Module.Write SpriteRotation=false
    Module.Write SpriteSize=false
    Position=&Add_Vector(A=&SimulationPosition(), B=<0 0 10>)
   )
   /SetVariables(Particles.MeshIndex=&RandomRangeInt(Minimum=0, Maximum=2)) {isSetParameters=true, module=SetVariables_#1}
   &CalculateSizeAndRotationalInertiaByMass(Depth=0, Height=5, Module.Manually Enter Density=false)
   &SkeletalMeshLocation(
    Bone Sampling Mode=::NewEnumerator4
    MaskSpawnGroup=false
    Position Sampling=::NewEnumerator1
    Sampled Position Offset=<0 0 20>
    Sampled Position Offset Coordinate Space:ENiagaraSkelMeshTransforms=::NewEnumerator2
    Skeletal Mesh=$User.New Skeletal Mesh
   )
   &SphereLocation(
    Hemisphere X=true
    MaskSpawnGroup=false
    Non Uniform Scale=<2.5 1 1>
    Override Seed=false
    Sphere Radius=2
    Surface Only=true
    UseNonUniformScale=true
    UseOffset=false
    UseSphereOrientation=false
   ) {enabled=false}
   &AddVelocity(
    Cone Angle=2
    Cone Axis=<0.4 -0.75 2>
    Module.UseConeFalloff=false
    Module.UseVelDistribution=false
    Override Seed=false
    Velocity Mode=::NewEnumerator2
    Velocity Speed=&FN_FrontEndFloatModifier(Gameplay=500, Front End=420)
   ) {module=AddVelocity002}
   &AddVelocity(Coordinate Space=::Simulation, Velocity=<150 -200 300>) {enabled=false}
   &AddVelocity(Velocity=&Multiply_VectorByFloat(
     Vector=&TransformVector(Vector=$Output.SphereLocation.SphereNormal, Source Space=::World, Destination Space=::Local)
     Float=&RandomRangeFloat(Minimum=40, Maximum=180)
    )) {enabled=false, module=AddVelocity001}
   &SubUVAnimation(SubUV Animation Mode=::NewEnumerator1) {enabled=false}
   /Niagara/Modules/Spawn/Orientation/V2/InitialMeshOrientation(
    Rotation=&RandomRangeVector(Minimum=<-1 -1 -1>, Maximum=<1 1 1>)
    Use Orientation Vector=true
    Use Rotation Vector=true
   ) {enabled=false}
   &AddRotationalVelocity(
    Coordinate Space:ENiagaraMeshTransforms=::NewEnumerator2
    Rotation Rate=&RandomRangeVector(Minimum=<-0.7 -0.7 -0.7>, Maximum=<0.7 0.7 0.7>)
   )
   &DynamicMaterialParameters(
    Index 0 Param 1=&RandomRangeFloat(Minimum=0, Maximum=1)
    Module.Param0WriteEnabled=true
    Module.Param1WriteEnabled=false
    Module.Param2WriteEnabled=false
    Module.Param3WriteEnabled=false
   )
   &InitialMeshOrientation(
    Orientation Coordinate Space:ENiagaraCoordinateSpace=::World
    Orientation Vector=$Particles.Velocity
    Use Orientation Vector=true
    Use Rotation Vector=false
   ) {enabled=false, module=InitialMeshOrientation001}
   /Niagara/Modules/Spawn/Orientation/V3/InitialMeshOrientation(Use Rotation Vector=false) {module=InitialMeshOrientation002}
  ]
  particleUpdateScript [
   &ParticleState()
   &ScaleMeshSize(Scale Factor=&MakeVector(
     X=&FloatFromCurve(FloatCurve=DataInterface() {propertyValues {Curve {keys [
          {arriveTangent=26.0828, interpMode=RCIM_Cubic, leaveTangent=26.0828, tangentMode=RCTM_User, time=0, value=0.2}
          {arriveTangent=0.0334, interpMode=RCIM_Cubic, leaveTangent=0.0334, tangentMode=RCTM_User, time=0.0394, value=4.0724}
          {arriveTangent=-2.5074, interpMode=RCIM_Cubic, leaveTangent=-2.5074, tangentMode=RCTM_User, time=0.1173, value=0.7442}
          {arriveTangent=-0.1609, interpMode=RCIM_Cubic, leaveTangent=-0.1609, tangentMode=RCTM_User, time=0.2152, value=1.1946}
          {arriveTangent=0.1391, interpMode=RCIM_Cubic, leaveTangent=0.1391, tangentMode=RCTM_User, time=0.3123, value=0.8985}
          <at 0.35 1>
         ]}}})
     Y=&FloatFromCurve(FloatCurve=DataInterface() {propertyValues {Curve {keys [
          <at -0.003 1.0812 RCIM_Cubic>
          {arriveTangent=3.927, interpMode=RCIM_Cubic, leaveTangent=3.927, time=0.0492, value=1.1749}
          {arriveTangent=-2.0352, interpMode=RCIM_Cubic, leaveTangent=-2.0352, time=0.1219, value=1.5716}
          {arriveTangent=-0.5182, interpMode=RCIM_Cubic, leaveTangent=-0.5182, tangentMode=RCTM_User, time=0.2148, value=0.8379}
          {arriveTangent=-0.6419, interpMode=RCIM_Cubic, leaveTangent=-0.6419, tangentMode=RCTM_User, time=0.3121, value=1.1071}
          <at 0.35 1 RCIM_Cubic>
         ]}}})
     Z=&FloatFromCurve(FloatCurve=DataInterface() {propertyValues {Curve {keys [
          <at -0.0016 1.0762 RCIM_Cubic>
          {arriveTangent=-0.5131, interpMode=RCIM_Cubic, leaveTangent=-0.5131, tangentMode=RCTM_User, time=0.0502, value=1.1706}
          {arriveTangent=-2.0432, interpMode=RCIM_Cubic, leaveTangent=-2.0432, time=0.1212, value=1.571}
          {arriveTangent=0.2115, interpMode=RCIM_Cubic, leaveTangent=0.2115, tangentMode=RCTM_User, time=0.2145, value=0.835}
          {arriveTangent=-0.2737, interpMode=RCIM_Cubic, leaveTangent=-0.2737, tangentMode=RCTM_User, time=0.312, value=1.1039}
          <at 0.35 1>
         ]}}})
    )) {module=ScaleMeshSize001}
   &GravityForce(Gravity=&RandomRangeVector(Minimum=<0 0 -500>, Maximum=<1 1 -1200>))
   &Drag(Drag=&RandomRangeFloat(Minimum=2, Maximum=3), Use Linear Drag=true, Use Rotational Drag=false)
   &ScaleColor(
    Scale Alpha=&FloatFromCurve(FloatCurve=DataInterface() {propertyValues {Curve {keys [<at -0.0046 1.0113 RCIM_Cubic>,<at 1.0011 0.002 RCIM_Cubic>]}}})
    Scale RGB=<0 0 0>
    ScaleA=true
    ScaleRGB=true
   )
   &self:SphereBlockingVolume(Pivot Offset:Vector3f=<0 0 10>, Sphere Origin:Vector3f=&SimulationPosition(), Sphere Radius=5) {enabled=false, module=SphereBlockingVolume}
   &Collision(
    Age Colliding Particles=false
    Control Roll On Collision=true
    Module.Control Rest State Via Surface Normal=false
    Module.EnableMaxCollisionCount=false
    Module.Randomize Collision Normal Vector=false
    Restitution=&RandomRangeFloat(Minimum=0.25, Maximum=0.6)
    ScaleRateOfRotation=false
   ) {enabled=false}
   &SolveForcesAndVelocity(Clamp Velocity=false, Limit Acceleration=false)
   &SolveRotationalForcesAndVelocity()
   &ScaleMeshSize(Scale Factor=&VectorFromFloat(Value=&FloatFromCurve(FloatCurve=DataInterface() {propertyValues {Curve {keys [<at 0.75 1 RCIM_Cubic>, {arriveTangent=-8.6407, interpMode=RCIM_Cubic, leaveTangent=-8.6407, tangentMode=RCTM_User, time=1, value=0}]}}})))
  ]
  rendererClasses [&NiagaraMeshRendererProperties, &NiagaraSpriteRendererProperties]
  renderers [
   {rendererClass=&NiagaraMeshRendererProperties, rendererIndex=0}
   {rendererClass=&NiagaraSpriteRendererProperties, rendererIndex=1}
   {rendererClass=&NiagaraMeshRendererProperties, rendererIndex=2}
  ]
 }
}
summary {
 systemName=NS_Emote_SweetToss_Burst
 emitters [
  {
   emitterName=SweetToss_Burst
   rendererClasses [&NiagaraMeshRendererProperties, &NiagaraSpriteRendererProperties]
  }
  {
   emitterName=SweetToss_Burst_Mobile
   rendererClasses [&NiagaraMeshRendererProperties, &NiagaraSpriteRendererProperties]
  }
  {emitterName=Empty, rendererClasses [&NiagaraSpriteRendererProperties]}
 ]
 userVariables [
  User.bIsFrontEnd {type=NiagaraBool}
  User.bIsFrontEndPreview {type=NiagaraBool}
  User.New Skeletal Mesh {
   type=NiagaraDataInterfaceSkeletalMesh
   defaultValue {
    struct=NiagaraExt_VariableValue_DataInterface
    value {dataInterface=&self:NiagaraDataInterfaceSkeletalMesh_0}
   }
  }
  User.spawnRatePopC {type=NiagaraInt32}
  User.Switch {type=NiagaraFloat, defaultValue=1, description=""}
 ]
}
