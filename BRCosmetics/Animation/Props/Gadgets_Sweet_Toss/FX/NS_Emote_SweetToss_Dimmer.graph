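# Niagara system for the Sweet Toss emote gadget prop, named as a Dimmer variant, with no emitters in its summary.
# Its only listed module is SystemState, so it shows no spawning or rendering logic of its own.
# It exposes user parameters for bounce intensity, colors, pour velocity and spawn intensity, plus a skeletal mesh data interface, with no emitters listed to consume them.
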NiagaraSystem /BRCosmetics/Animation/Props/Gadgets_Sweet_Toss/FX/NS_Emote_SweetToss_Dimmer
dependencies {usedDataInterfaces [], usedDynamicInputs [], usedModules [&SystemState]}
summary {
 systemName=NS_Emote_SweetToss_Dimmer
 emitters []
 userVariables [
  User.BounceIntensity {type=NiagaraFloat, defaultValue=0, description=""}
  User.Colors {type=NiagaraFloat, defaultValue=0, description=""}
  User.New Skeletal Mesh {
   type=NiagaraDataInterfaceSkeletalMesh
   defaultValue {
    struct=NiagaraExt_VariableValue_DataInterface
    value {dataInterface=&self:NiagaraDataInterfaceSkeletalMesh_0}
   }
  }
  User.PourVelocity {type=NiagaraFloat, defaultValue=0, description=""}
  User.SpawnIntensity {type=NiagaraFloat, defaultValue=0.2, description=""}
 ]
}
