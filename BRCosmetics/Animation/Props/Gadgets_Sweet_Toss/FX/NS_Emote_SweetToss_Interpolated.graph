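# Niagara effect for the Sweet Toss gadget emote: a GPU emitter fires three quick small bursts of sweets as sprites and randomly indexed meshes.
# Each particle interpolates its position between two skeletal mesh sample points over a short curve, following the animated mesh rather than simulating a throw.
# User.Colors selects one of three bands in a stepped random palette.
# Velocity, gravity, drag and collision modules remain but are disabled.
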NiagaraSystem /BRCosmetics/Animation/Props/Gadgets_Sweet_Toss/FX/NS_Emote_SweetToss_Interpolated
dependencies {
 usedDataInterfaces [&NiagaraDataInterfaceColorCurve, &NiagaraDataInterfaceArrayFloat, &NiagaraDataInterfaceCurve]
 usedDynamicInputs []
 usedModules [&SystemState]
}
emitters {Cereal_World002 {
  simTarget=GPUComputeSim
  emitterUpdateScript [
   &EmitterState()
   &SpawnRate(
    SpawnRate=&Lerp_Float(A=&Multiply_Float(A=$User.SpawnIntensity, B=300), B=300, Alpha=1)
    Use Spawn Probability=false
   ) {enabled=false, module=SpawnRate001}
   &SpawnBurst_Instantaneous(Spawn Count=2, Use Spawn Probability=false, UseLoopCountLimit=false)
   &SpawnBurst_Instantaneous(Spawn Count=2, Spawn Time=0.1, Use Spawn Probability=false, UseLoopCountLimit=false) {module=SpawnBurst_Instantaneous001}
   &SpawnBurst_Instantaneous(Spawn Time=0.2, Use Spawn Probability=false, UseLoopCountLimit=false) {module=SpawnBurst_Instantaneous002}
  ]
  particleSpawnScript [
   &InitializeParticle(
    Color=&ColorFromCurve(
     ColorCurve=DataInterface() {propertyValues {
       AlphaCurve {keys [<at 0 1 RCIM_Constant>]}
       BlueCurve {keys [<at 0 1 RCIM_Constant>,<at 0.1739 0.0551 RCIM_Constant>,<at 0.2302 0.4418 RCIM_Constant>,<at 0.2857 0.9601 RCIM_Constant>,<at 0.33 0 RCIM_Constant>,<at 0.5399 0 RCIM_Constant>,<at 0.5978 0.6094 RCIM_Constant>,<at 0.66 0.533 RCIM_Constant>,<at 0.8835 0.13 RCIM_Constant>,<at 0.9481 1 RCIM_Constant>]}
       GreenCurve {keys [<at 0 0.8658 RCIM_Constant>,<at 0.1739 0.3212 RCIM_Constant>,<at 0.2302 0.0248 RCIM_Constant>,<at 0.2857 0.015 RCIM_Constant>,<at 0.33 0.5573 RCIM_Constant>,<at 0.5399 0.2124 RCIM_Constant>,<at 0.5978 0.836 RCIM_Constant>,<at 0.66 0.048 RCIM_Constant>,<at 0.8835 0.0009 RCIM_Constant>,<at 0.9481 0.54 RCIM_Constant>]}
       RedCurve {keys [<at 0 0 RCIM_Constant>,<at 0.1739 0.0139 RCIM_Constant>,<at 0.2302 0.8388 RCIM_Constant>,<at 0.2857 0.094 RCIM_Constant>,<at 0.33 1 RCIM_Constant>,<at 0.5399 1 RCIM_Constant>,<at 0.5978 1 RCIM_Constant>,<at 0.66 0.1225 RCIM_Constant>,<at 0.8835 0.5365 RCIM_Constant>,<at 0.9481 0.0381 RCIM_Constant>]}
      }}
     CurveIndex=&RandomRangeFloat(
      Minimum=&SelectFloatFromArray(
       Direct Array Index=&MakeIntFromFloat(Float=&Add_Float(A=$User.Colors, B=0))
       Direct Array Mode=::NewEnumerator1
       Float Selection Array=DataInterface() {propertyValues {FloatData [0,0.33,0.66]}}
      )
      Maximum=&SelectFloatFromArray(
       Direct Array Index=&MakeIntFromFloat(Float=&Add_Float(A=$User.Colors, B=0))
       Direct Array Mode=::NewEnumerator1
       Float Selection Array=DataInterface() {propertyValues {FloatData [0.32,0.65,1]}}
      )
     )
    )
    Lifetime=1.2
    Mesh Scale Mode=::NewEnumerator3
    Mesh Uniform Scale=0.8
    Sprite Rotation Mode=::NewEnumerator2
    Sprite Size Mode=::NewEnumerator3
    Uniform Sprite Size=&RandomRangeFloat(Minimum=2, Maximum=7)
    UsePositionOffset=false
    Write Material Random=true
   )
   &BoxLocation(Box Offset=<0 0 25>, Box Size=<25 5 10>, MaskSpawnGroup=false, Module.Override Seed=false, Rotate Box=false, Surface Only=false) {enabled=false}
   &AddVelocityFromPoint(
    Module.Use Falloff Distance=false
    Origin Offset=<1 0 -10>
    Origin Offset Coordinate Space=::Simulation
    Velocity Origin:NiagaraPosition=$Particles.SkeletalMeshLocation.SampledPosition
    Velocity Strength=&Multiply_Float(A=100, B=$User.SpawnIntensity)
   ) {enabled=false}
   &AddVelocity(
    Coordinate Space=::Simulation
    Velocity=&Multiply_VectorByFloat(
     Vector=&RotateVectorByQuaternion(
      Vector To Rotate=&RandomRangeVector(Minimum=<0 0 150>, Maximum=<0 0 300>)
      Quaternion=$Particles.SkeletalMeshLocation.SampledRotation
     )
     Float=$User.SpawnIntensity
    )
   ) {enabled=false}
   /Niagara/Modules/Update/SubUV/SubUVAnimation(SubUV Animation Mode=::NewEnumerator1)
   &InheritVelocity(
    Inherited Velocity Amount Scale=<0.75 0.75 0.75>
    Speed Limit=false
    UseSpeedThreshold=false
    Velocity Source=$Particles.SkeletalMeshLocation.SampledVelocity
   ) {enabled=false}
   &InitialMeshOrientation(Use Rotation Vector=false)
   /SetVariables(Particles.Position=&Lerp_Vector(A=Linked() {linkedVariable {name=Particles.Position, type=NiagaraPosition}}, B=<1 1 1>, Alpha=0.5)) {isSetParameters=true}
   /SetVariables(Particles.MeshIndex=&RandomRangeInt(Minimum=0, Maximum=2)) {isSetParameters=true}
  ]
  particleUpdateScript [
   &GravityForce(Gravity=<0 0 -750>) {enabled=false}
   &SkeletalMeshLocation(
    Bone Sampling Mode=::NewEnumerator4
    MaskSpawnGroup=false
    Orientation Sampling=::NewEnumerator1
    Position Sampling=::NewEnumerator2
    Sampled Position Offset=<5 -3 -9>
    Skeletal Mesh=$User.New Skeletal Mesh
    Use Sampled Mesh Attribute as Ribbon ID=false
   ) {module=SkeletalMeshLocation001}
   &SkeletalMeshLocation(
    Bone / Socket Index=2
    Bone Sampling Mode=::NewEnumerator4
    MaskSpawnGroup=false
    Orientation Sampling=::NewEnumerator1
    Position Sampling=::NewEnumerator2
    Sampled Position Offset=&RandomRangeVector(Minimum=<-5 3 6>, Maximum=<-10 3 -5>)
    Skeletal Mesh=$User.New Skeletal Mesh
    Use Sampled Mesh Attribute as Ribbon ID=false
   )
   &Drag(
    Drag=&Lerp_Float(
     A=1
     B=0.7
     Alpha=&NormalizeFloat(Value=&MakeFloatFromVector2d(Vector 2D=$Particles.Initial.SpriteSize), Min=2, Max=7)
    )
    Use Linear Drag=true
    Use Rotational Drag=false
   ) {enabled=false}
   &Collision(
    Age Colliding Particles=false
    GPU Collision Type=::NewEnumerator2
    Module.Control Rest State Via Surface Normal=false
    Module.EnableMaxCollisionCount=false
    Module.Randomize Collision Normal Vector=true
    Randomize Collision Normal=0.25
    Restitution=&RandomRangeFloat(Minimum=0.25, Maximum=0.6)
   ) {enabled=false}
   /SetVariables(Particles.Position=&Lerp_Vector(
     A=Linked() {linkedVariable {name=Particles.SkeletalMeshLocation001.SampledPosition, type=NiagaraPosition}}
     B=Linked() {linkedVariable {name=Particles.SkeletalMeshLocation.SampledPosition, type=NiagaraPosition}}
     Alpha=&FloatFromCurve(FloatCurve=DataInterface() {propertyValues {Curve {keys [<at 0 0>,<at 0.2 1>]}}})
    )) {isSetParameters=true}
   &SolveForcesAndVelocity(Clamp Velocity=false, Limit Acceleration=false)
   &ParticleState()
   &ScaleSpriteSize(Uniform Curve Sprite Scale=DataInterface() {propertyValues {Curve {keys [{interpMode=RCIM_Cubic, tangentMode=RCTM_User, time=0, value=1}, {interpMode=RCIM_Cubic, tangentMode=RCTM_User, time=1, value=0}]}}})
   &SpriteRotationRate()
  ]
  rendererClasses [&NiagaraSpriteRendererProperties, &NiagaraMeshRendererProperties]
  renderers [
   {rendererClass=&NiagaraSpriteRendererProperties}
   {rendererClass=&NiagaraMeshRendererProperties, rendererIndex=1}
  ]
 }}
summary {
 systemName=NS_Emote_SweetToss_Interpolated
 emitters [{
   emitterName=Cereal_World002
   simTarget=GPUComputeSim
   rendererClasses [&NiagaraSpriteRendererProperties, &NiagaraMeshRendererProperties]
  }]
 userVariables [
  User.BounceIntensity {type=NiagaraFloat, defaultValue=0, description=""}
  User.Colors {type=NiagaraFloat, defaultValue=0, description=""}
  User.New Skeletal Mesh {
   type=NiagaraDataInterfaceSkeletalMesh
   defaultValue {
    struct=NiagaraExt_VariableValue_DataInterface
    value {dataInterface=&self:NiagaraDataInterfaceSkeletalMesh_0}
   }
  }
  User.SpawnIntensity {type=NiagaraFloat, defaultValue=0.2, description=""}
 ]
}
